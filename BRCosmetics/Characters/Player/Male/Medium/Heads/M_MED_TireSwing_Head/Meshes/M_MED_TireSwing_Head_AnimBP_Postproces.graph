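# Post-process AnimBP for the TireSwing male head, running a head Control Rig over the input pose with hat_on/off as alpha.
# Initialization and customization-loaded events set Is_hemlet from the TireSwing cosmetic meta tag.
# Each update, an emote, front-end and cine-posing check can set the alpha, but a later Is_hemlet branch overwrites it to on or off.
# It also reads the "Hide Element 02" material parameter without using the result.
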AnimBlueprint /BRCosmetics/Characters/Player/Male/Medium/Heads/M_MED_TireSwing_Head/Meshes/M_MED_TireSwing_Head_AnimBP_Postproces
target_skeleton=/Game/Characters/Player/Common/Fortnite_Base_Head/Fortnite_Base_3L_Head_Skeleton
graphs [
 AnimGraph {nodes [
   n1 = Control Rig(Source=Input Pose(), Alpha=Get hat_on/off() -> (hat_on/off:double)) {
    alphaBoolBlend {blendOption=Linear}
    controlRigAssetReference {blueprintRigClass=&/M_MED_TireSwing_Head_CtrlRig.M_MED_TireSwing_Head_CtrlRig_C}
   }
   Output Pose(Result=@n1)
  ]}
 EventGraph {nodes [
   Event BlueprintUpdateAnimation()
   > n1 = Sequence!(execute) -> (then_0, then_1, then_2:Exec, then_3:Exec, then_4:Exec)
   > n2 = Is Valid(InputObject=TryGetPawnOwner())
   n3 = Set CinePosing(execute=@n2.Is Not Valid, CinePosing:bool=true) -> (Output_Get:bool)
   n4 = Get bIsPlayingEmote()
   n5 = Get bIsInFrontEnd()
   n6 = Get CinePosing() -> (CinePosing:bool)
   n7 = Branch(execute=[@n3.then, @n8.then], Condition=AND Boolean!(A=NOT Boolean(A=@n4), B=NOT Boolean(A=@n5), C:bool=NOT Boolean(A=@n6)) -> (ReturnValue))
   Set hat_on/off(execute=[@n10.then, @n7.then], hat_on/off:double=1) -> (Output_Get:double)
   n8 = Set CinePosing(execute=@n2.Is Valid, CinePosing=false)
   n9 = Get Dynamic Material Instance(execute=@n1.then_1, SkeletalMeshComponent=GetPartSkeletalMeshComponent())
   > Get Scalar Parameter Value!(execute, self:Material Instance Dynamic Object Reference=@n9, ParameterName:Name=Hide Element 02) -> (then, ReturnValue:float) {is=title K2Node_CallFunction}
   n10 = Branch(execute=@n1.then_2, Condition=Get Is_hemlet() -> (Is_hemlet:bool))
   Set hat_on/off(execute=@n10.else, hat_on/off=0)
   n11 = Event OnCharacterCustomizationLoaded()
   Set Is_hemlet(
    execute=[@n11.then, @n12.then]
    Is_hemlet:bool=CharacterItemDefinitionHasAnyMetaTag(Tags=Make Array("[0]"=(TagName="Cosmetics.MetaTags.TireSwing")))
   ) -> (Output_Get:bool)
   n12 = Event BlueprintInitializeAnimation()
  ]}
]
variables [
 CinePosing {type=boolean}
 hat_on/off {type=number}
 hat_on/off_0 {type=boolean}
 head rot {type=Rotator}
 HeadLimit {type=boolean, default=true}
 Is_hemlet {type=boolean}
 X percentage {type=number}
 Y percentage {type=number}
 Z percentage {type=number, default=0.8}
]
